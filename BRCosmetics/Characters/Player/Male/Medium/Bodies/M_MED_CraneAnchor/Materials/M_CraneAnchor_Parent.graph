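# Parent material for the CraneAnchor male medium player body cosmetic, built on the shared 2023 character parent functions with material attributes.
# The PreFX function builds the base surface, and the "Use Sequins" static switch optionally swaps in the weapon-skin MF_Sequins function, fed by the Tech Art Mask.
# The result feeds the PostFX function for final output.
# Shading model comes from the graph, with usage enabled for skeletal, morph, deformer, Nanite and Niagara meshes.
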Material /BRCosmetics/Characters/Player/Male/Medium/Bodies/M_MED_CraneAnchor/Materials/M_CraneAnchor_Parent
graph {
 properties {
  bUseMaterialAttributes=true
  bUsedWithMeshDeformer=true
  bUsedWithMorphTargets=true
  bUsedWithNanite=true
  bUsedWithNiagaraMeshParticles=true
  bUsedWithSkeletalMesh=true
  shadingModel=MSM_FromMaterialExpression
  usedShadingModels=Default Lit | Subsurface | Clear Coat | Subsurface Profile | Cloth
 }
 nodes [
  n1 = MaterialFunctionCall() {materialFunction=&MF_CharacterParent_2023_PreFX}
  MaterialFunctionCall(
   Material=StaticSwitchParameter(
    True=MaterialFunctionCall!(Mask, TechArtMask=@n1.Tech Art Mask, inMat=@n1.Tech Art Mask) {materialFunction=/Game/Weapons/WeaponSkins/MaterialFunction/MF_Sequins}.Result
    False=@n1.Result
   ) {group=04_Tech Art, parameterName=Use Sequins}
   M=@n1.M
   Pre Skinned Local Position=@n1.M
   Distance Blend=@n1.M
   Fresnel=@n1.M
  ) {materialFunction=&MF_CharacterParent_2023_PostFX}
 ]
}
