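# Post-processes the CraneAnchor male body cosmetic's pose to drive its jacket, necklace and tape secondary motion.
# Pose copied from the parent mesh passes through a Control Rig, leg-tracking Look Ats and pelvis-space RigidBody simulation.
# Pelvis constraints and additive bone corrections then adjust the jacket, before Anim Dynamics swings the necklace and tape.
# The event graph derives alphas from thigh angles, movement states and frame time, plus a state-based gravity override.
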AnimBlueprint /BRCosmetics/Characters/Player/Male/Medium/Bodies/M_MED_CraneAnchor/Meshes/M_MED_CraneAnchor_AnimBP
target_skeleton=&/M_MED_CraneAnchor_Skeleton
graphs [
 AnimGraph {nodes [
   n1 = Get FloatConstraint_alpha() -> (FloatConstraint_alpha:double)
   n2 = Control Rig(Source=Copy Pose From Mesh() {bUseAttachedParent=true}) {
    alphaBoolBlend {blendOption=Linear}
    controlRigAssetReference {blueprintRigClass=&/M_MED_CraneAnchor_CtrlRig.M_MED_CraneAnchor_CtrlRig_C}
   }
   n3 = Look At - Bone: jacket_front_base_L(ComponentPose=Local To Component(LocalPose=@n2), Alpha=1) {
    alphaBoolBlend {blendOption=Linear}
    boneToModify {boneName=jacket_front_base_L}
    lookAtLocation=<0 10>
    lookAtTarget {boneReference {boneName=calf_l}}
    lookAt_Axis {axis=<1 0>}
   }
   n4 = Look At - Bone: jacket_base_L(ComponentPose=@n3, Alpha=0.7) {
    alphaBoolBlend {blendOption=Linear}
    boneToModify {boneName=jacket_base_L}
    lookAtLocation=<0 10>
    lookAtTarget {boneReference {boneName=calf_l}}
    lookAt_Axis {axis=<1 0>}
   }
   n5 = "Transform (Modify) Bone - Bone: jacket_front_base_L"(
    ComponentPose=@n4
    Alpha=Get LeftLegForwardBendCompensationAlpha() -> (LeftLegForwardBendCompensationAlpha:double)
    Translation="0,0,0"
    Rotation="-15,0,-15"
    Scale="1,1,1"
   ) {
    rotationMode=BMM_Additive
    rotationSpace=BCS_BoneSpace
    translationMode=BMM_Additive
    translationSpace=BCS_BoneSpace
    alphaBoolBlend {blendOption=Linear}
    alphaScaleBiasClamp {bInterpResult=true}
    boneToModify {boneName=jacket_front_base_L}
    rotation {pitch=-15, roll=-15}
   }
   n6 = Look At - Bone: jacket_front_base_R(ComponentPose=@n5, Alpha=1) {
    alphaBoolBlend {blendOption=Linear}
    boneToModify {boneName=jacket_front_base_R}
    lookAtLocation=<0 -10>
    lookAtTarget {boneReference {boneName=calf_r}}
    lookAt_Axis {axis=<1 0>}
   }
   n7 = Look At - Bone: jacket_base_R(ComponentPose=@n6, Alpha=0.7) {
    alphaBoolBlend {blendOption=Linear}
    boneToModify {boneName=jacket_base_R}
    lookAtLocation=<0 -10>
    lookAtTarget {boneReference {boneName=calf_r}}
    lookAt_Axis {axis=<1 0>}
   }
   n8 = RigidBody!(ComponentPose=@n7, Alpha, OverrideWorldGravity:Vector=Get Gravity Override(), ExternalForce) -> (Pose) {
    bClampLinearTranslationLimitToRefPose=true
    bOverrideWorldGravity=true
    overridePhysicsAsset=&/M_MED_CraneAnchor_Physics
    simulationSpace=BaseBoneSpace
    alphaBoolBlend {blendOption=Linear}
    alphaScaleBiasClamp {bInterpResult=true, bMapRange=true, interpSpeedDecreasing=15, interpSpeedIncreasing=20, outRange {max=0.6, min=1}}
    baseBoneRef {boneName=pelvis}
    componentAppliedLinearAccClamp=<1000 1000 1000>
    componentLinearAccScale=<0.5 0.5 0.5>
    componentLinearVelScale=<0.6 0.6 0.6>
    overrideWorldGravity {z=-980}
   }
   n9 = "Transform (Modify) Bone - Bone: dyn_jacket_front_02_L"(ComponentPose=@n8, Alpha=Get ForwardMovementAlpha() -> (ForwardMovementAlpha:double), Translation="0,0,-2", Rotation="0,0,0", Scale="1,1,1") {
    rotationMode=BMM_Additive
    rotationSpace=BCS_BoneSpace
    translationMode=BMM_Additive
    translationSpace=BCS_BoneSpace
    alphaBoolBlend {blendOption=Linear}
    alphaScaleBiasClamp {bInterpResult=true}
    boneToModify {boneName=dyn_jacket_front_02_L}
    translation {z=-2}
   }
   n10 = "Constraint - dyn_jacket_01_L (P)"(ComponentPose=@n9, Alpha=@n1, ConstraintWeights_0=1) {
    alphaBoolBlend {blendOption=Linear}
    alphaScaleBiasClamp {bInterpResult=true}
    boneToModify {boneName=dyn_jacket_01_L}
    constraintSetup [{targetBone {boneName=pelvis}}]
    constraintWeights [1]
   }
   n11 = "Constraint - dyn_jacket_01_R (P)"(ComponentPose=@n10, Alpha=@n1, ConstraintWeights_0=1) {
    alphaBoolBlend {blendOption=Linear}
    alphaScaleBiasClamp {bInterpResult=true}
    boneToModify {boneName=dyn_jacket_01_R}
    constraintSetup [{targetBone {boneName=pelvis}}]
    constraintWeights [1]
   }
   n12 = "Constraint - dyn_jacket_front_01_L (P)"(ComponentPose=@n11, Alpha=@n1, ConstraintWeights_0=1) {
    alphaBoolBlend {blendOption=Linear}
    alphaScaleBiasClamp {bInterpResult=true}
    boneToModify {boneName=dyn_jacket_front_01_L}
    constraintSetup [{targetBone {boneName=pelvis}}]
    constraintWeights [1]
   }
   n13 = "Constraint - dyn_jacket_front_01_R (P)"(ComponentPose=@n12, Alpha=@n1, ConstraintWeights_0=1) {
    alphaBoolBlend {blendOption=Linear}
    alphaScaleBiasClamp {bInterpResult=true}
    boneToModify {boneName=dyn_jacket_front_01_R}
    constraintSetup [{targetBone {boneName=pelvis}}]
    constraintWeights [1]
   }
   n14 = "Constraint - dyn_jacket_back_01_L (P)"(ComponentPose=@n13, Alpha=@n1, ConstraintWeights_0=1) {
    alphaBoolBlend {blendOption=Linear}
    alphaScaleBiasClamp {bInterpResult=true}
    boneToModify {boneName=dyn_jacket_back_01_L}
    constraintSetup [{targetBone {boneName=pelvis}}]
    constraintWeights [1]
   }
   n15 = "Constraint - dyn_jacket_back_01_R (P)"(ComponentPose=@n14, Alpha=@n1, ConstraintWeights_0=1) {
    alphaBoolBlend {blendOption=Linear}
    alphaScaleBiasClamp {bInterpResult=true}
    boneToModify {boneName=dyn_jacket_back_01_R}
    constraintSetup [{targetBone {boneName=pelvis}}]
    constraintWeights [1]
   }
   n16 = "Constraint - dyn_jacket_02_L (P)"(ComponentPose=@n15, Alpha=@n1, ConstraintWeights_0=1) {
    alphaBoolBlend {blendOption=Linear}
    alphaScaleBiasClamp {bInterpResult=true}
    boneToModify {boneName=dyn_jacket_02_L}
    constraintSetup [{targetBone {boneName=pelvis}}]
    constraintWeights [1]
   }
   n17 = "Constraint - dyn_jacket_02_R (P)"(ComponentPose=@n16, Alpha=@n1, ConstraintWeights_0=1) {
    alphaBoolBlend {blendOption=Linear}
    alphaScaleBiasClamp {bInterpResult=true}
    boneToModify {boneName=dyn_jacket_02_R}
    constraintSetup [{targetBone {boneName=pelvis}}]
    constraintWeights [1]
   }
   n18 = "Constraint - dyn_jacket_front_02_L (P)"(ComponentPose=@n17, Alpha=@n1, ConstraintWeights_0=1) {
    alphaBoolBlend {blendOption=Linear}
    alphaScaleBiasClamp {bInterpResult=true}
    boneToModify {boneName=dyn_jacket_front_02_L}
    constraintSetup [{targetBone {boneName=pelvis}}]
    constraintWeights [1]
   }
   n19 = "Constraint - dyn_jacket_front_02_R (P)"(ComponentPose=@n18, Alpha=@n1, ConstraintWeights_0=1) {
    alphaBoolBlend {blendOption=Linear}
    alphaScaleBiasClamp {bInterpResult=true}
    boneToModify {boneName=dyn_jacket_front_02_R}
    constraintSetup [{targetBone {boneName=pelvis}}]
    constraintWeights [1]
   }
   n20 = "Constraint - dyn_jacket_back_02_L (P)"(ComponentPose=@n19, Alpha=@n1, ConstraintWeights_0=1) {
    alphaBoolBlend {blendOption=Linear}
    alphaScaleBiasClamp {bInterpResult=true}
    boneToModify {boneName=dyn_jacket_back_02_L}
    constraintSetup [{targetBone {boneName=pelvis}}]
    constraintWeights [1]
   }
   n21 = "Constraint - dyn_jacket_back_02_R (P)"(ComponentPose=@n20, Alpha=@n1, ConstraintWeights_0=1) {
    alphaBoolBlend {blendOption=Linear}
    alphaScaleBiasClamp {bInterpResult=true}
    boneToModify {boneName=dyn_jacket_back_02_R}
    constraintSetup [{targetBone {boneName=pelvis}}]
    constraintWeights [1]
   }
   n22 = "Transform (Modify) Bone - Bone: dyn_jacket_front_02_R"(ComponentPose=@n21, Alpha=1, Translation="0,0,1.5", Rotation="0,0,-15", Scale="1,1,1") {
    rotationMode=BMM_Additive
    translationMode=BMM_Additive
    translationSpace=BCS_BoneSpace
    alphaBoolBlend {blendOption=Linear}
    alphaScaleBiasClamp {bInterpResult=true}
    boneToModify {boneName=dyn_jacket_front_02_R}
    rotation {roll=-15}
    translation {z=1.5}
   }
   n23 = Look At - Bone: jacket_front_base_R(ComponentPose=@n22, Alpha=1) {
    alphaBoolBlend {blendOption=Linear}
    alphaScaleBiasClamp {bInterpResult=true}
    boneToModify {boneName=jacket_front_base_R}
    lookAtLocation=<30 -6.7769>
    lookAtTarget {boneReference {boneName=thigh_r}}
    lookAt_Axis {axis=<1 0>}
   }
   n24 = "Transform (Modify) Bone - Bone: dyn_jacket_front_02_L"(ComponentPose=@n23, Alpha=1, Translation="0,0,1.5", Rotation="0,0,-15", Scale="1,1,1") {
    rotationMode=BMM_Additive
    translationMode=BMM_Additive
    translationSpace=BCS_BoneSpace
    alphaBoolBlend {blendOption=Linear}
    alphaScaleBiasClamp {bInterpResult=true}
    boneToModify {boneName=dyn_jacket_front_02_L}
    rotation {roll=-15}
    translation {z=1.5}
   }
   n25 = Look At - Bone: jacket_front_base_L(ComponentPose=@n24, Alpha=1) {
    alphaBoolBlend {blendOption=Linear}
    alphaScaleBiasClamp {bInterpResult=true}
    boneToModify {boneName=jacket_front_base_L}
    lookAtLocation=<-30 6.3191 0>
    lookAtTarget {boneReference {boneName=thigh_l}}
    lookAt_Axis {axis=<1 0>}
   }
   n26 = "Transform (Modify) Bone - Bone: jacket_back_base_L"!(ComponentPose=@n25, bAlphaBoolEnabled:bool=True, Translation="0,0,-4", Rotation="0,0,0", Scale="1,1,1") -> (Pose) {
    alphaInputType=Bool
    rotationMode=BMM_Additive
    rotationSpace=BCS_BoneSpace
    translationMode=BMM_Additive
    translationSpace=BCS_ParentBoneSpace
    alphaBoolBlend {blendInTime=1, blendOption=Linear, blendOutTime=1}
    boneToModify {boneName=jacket_back_base_L}
    translation {z=-4}
   }
   n27 = "Transform (Modify) Bone - Bone: dyn_jacket_front_02_R"(ComponentPose=@n26, Alpha=1, Translation="0,0,0", Rotation="0,0,60", Scale="1,1,1") {
    rotationMode=BMM_Additive
    translationMode=BMM_Additive
    translationSpace=BCS_BoneSpace
    alphaBoolBlend {blendOption=Linear}
    alphaScaleBiasClamp {bInterpResult=true, interpSpeedDecreasing=30, interpSpeedIncreasing=30}
    boneToModify {boneName=dyn_jacket_front_02_R}
    rotation {roll=60}
   }
   n28 = "Transform (Modify) Bone - Bone: dyn_jacket_front_02_L"(ComponentPose=@n27, Alpha=1, Translation="0,0,0", Rotation="0.000,0.000,60.000", Scale="1,1,1") {
    rotationMode=BMM_Additive
    translationMode=BMM_Additive
    translationSpace=BCS_BoneSpace
    alphaBoolBlend {blendOption=Linear}
    alphaScaleBiasClamp {bInterpResult=true, interpSpeedDecreasing=30, interpSpeedIncreasing=30}
    boneToModify {boneName=dyn_jacket_front_02_L}
    rotation {roll=60}
   }
   n29 = "Transform (Modify) Bone - Bone: dyn_jacket_front_01_L"(ComponentPose=@n28, Alpha=1, Translation="0,0,0", Rotation="0.000,0.000,10.000", Scale="1,1,1") {
    rotationMode=BMM_Additive
    translationMode=BMM_Additive
    translationSpace=BCS_BoneSpace
    alphaBoolBlend {blendOption=Linear}
    alphaScaleBiasClamp {bInterpResult=true, interpSpeedDecreasing=30, interpSpeedIncreasing=30}
    boneToModify {boneName=dyn_jacket_front_01_L}
    rotation {roll=10}
   }
   n30 = "Transform (Modify) Bone - Bone: jacket_base_R"(ComponentPose=@n29, Alpha=1, Translation="3,0,0", Rotation="0,0,-60", Scale="1,1,1") {
    rotationMode=BMM_Additive
    rotationSpace=BCS_BoneSpace
    translationMode=BMM_Additive
    translationSpace=BCS_ParentBoneSpace
    alphaBoolBlend {blendOption=Linear}
    boneToModify {boneName=jacket_base_R}
    rotation {roll=-60}
    translation {x=3}
   }
   n31 = "Transform (Modify) Bone - Bone: jacket_front_base_R"(ComponentPose=@n30, Alpha=1, Translation="1,0,0", Rotation="0,0,-30", Scale="1,1,1") {
    rotationMode=BMM_Additive
    rotationSpace=BCS_BoneSpace
    translationMode=BMM_Additive
    translationSpace=BCS_ParentBoneSpace
    alphaBoolBlend {blendOption=Linear}
    boneToModify {boneName=jacket_front_base_R}
    rotation {roll=-30}
    translation {x=1}
   }
   n32 = "Transform (Modify) Bone - Bone: jacket_front_base_R"(ComponentPose=@n31, Alpha=1, Translation="1,0,0", Rotation="-10,-10,0", Scale="1,1,1") {
    rotationMode=BMM_Additive
    rotationSpace=BCS_BoneSpace
    translationMode=BMM_Additive
    translationSpace=BCS_ParentBoneSpace
    alphaBoolBlend {blendOption=Linear}
    alphaScaleBiasClamp {bInterpResult=true}
    boneToModify {boneName=jacket_front_base_R}
    rotation {pitch=-10, yaw=-10}
    translation {x=1}
   }
   n33 = "Transform (Modify) Bone - Bone: jacket_base_L"(ComponentPose=@n32, Alpha=1, Translation="0,0,-1", Rotation="-10,0,0", Scale="1,1,1") {
    rotationMode=BMM_Additive
    rotationSpace=BCS_BoneSpace
    translationMode=BMM_Additive
    translationSpace=BCS_ParentBoneSpace
    alphaBoolBlend {blendOption=Linear}
    alphaScaleBiasClamp {bInterpResult=true, interpSpeedDecreasing=30}
    boneToModify {boneName=jacket_base_L}
    rotation {pitch=-10}
    translation {z=-1}
   }
   n34 = "Transform (Modify) Bone - Bone: jacket_front_base_L"(ComponentPose=@n33, Alpha=1, Translation="0,0,-2", Rotation="0,0,15", Scale="1,1,1") {
    rotationMode=BMM_Additive
    rotationSpace=BCS_BoneSpace
    translationMode=BMM_Additive
    translationSpace=BCS_ParentBoneSpace
    alphaBoolBlend {blendOption=Linear}
    alphaScaleBiasClamp {bInterpResult=true}
    boneToModify {boneName=jacket_front_base_L}
    rotation {roll=15}
    translation {z=-2}
   }
   n35 = Anim Dynamics - Bone: dyn_necklace_01!(ComponentPose=@n34, Alpha=1, GravityOverride:Vector="0,0,-900", ExternalForce="0,0,0") -> (Pose) {
    angularDampingOverride=0.9999
    bOverrideAngularDamping=true
    bUseGravityOverride=true
    alphaBoolBlend {blendOption=Linear}
    boundBone {boneName=dyn_necklace_01}
    componentAppliedLinearAccClamp=<1500 1500 1500>
    componentLinearAccScale=<0.15 0.15 0.15>
    componentLinearVelScale=<0.15 0.15 0.15>
    gravityOverride {z=-900}
    physicsBodyDefinitions [{
      boundBone {boneName=dyn_necklace_01}
      boxExtents=<20 15 20>
      constraintSetup {angularLimitsMax=<0 10 -2>, angularLimitsMin=<0 -10 -90>, angularTarget=<0 0 0>}
      localJointOffset=<5 0 0>
     }]
    relativeSpaceBone {boneName=spine_03}
   }
   n36 = Anim Dynamics - Bone: dyn_tape!(ComponentPose=@n35, Alpha=1, GravityOverride:Vector="0,0,-900", ExternalForce="0,0,0") -> (Pose) {
    angularDampingOverride=0.9999
    bOverrideAngularDamping=true
    bUseGravityOverride=true
    alphaBoolBlend {blendOption=Linear}
    boundBone {boneName=dyn_tape}
    componentAppliedLinearAccClamp=<1500 1500 1500>
    componentLinearAccScale=<0.15 0.15 0.15>
    componentLinearVelScale=<0.15 0.15 0.15>
    gravityOverride {z=-900}
    physicsBodyDefinitions [{
      boundBone {boneName=dyn_tape}
      boxExtents=<20 15 20>
      constraintSetup {angularLimitsMax=<0 10 0>, angularLimitsMin=<0 -10 -60>, angularTarget=<0 0 0>}
      localJointOffset=<5 0 0>
     }]
    relativeSpaceBone {boneName=spine_03}
   }
   Output Pose(Result=Component To Local(ComponentPose=@n36)) {layerGroup=None}
  ]}
 EventGraph {nodes [
   n1 = Event BlueprintUpdateAnimation()
   > n2 = Branch(Condition=Get bIsInFrontEnd())
   > n3 = Set FPS Compensation Alpha(FPS Compensation Alpha:double=0.3) -> (Output_Get:double)
   n4 = Sequence!(execute=[@n3.then, @n17.then]) -> (then_0, then_1, then_2:Exec, then_3:Exec)
   Is Valid(exec=@n4.then_2, InputObject=Get SkelMesh())
   > n5 = Gravity Override State Function DEPRECATED(
    CharacterPartAnimInstance=Self-Reference()
    GameTimeInSeconds=GetGameTimeInSeconds()
    GravityOverrideParams=Get Gravity Override Params()
    RBAN Params=Get RBAN Params()
   )
   > Set Gravity Override(Gravity Override=@n5)
   n6 = SelectVector(A="0.1,0.35,0.35", B="1,1,0.6", bPickA=OR Boolean(A=Get bIsSkyDiving(), B=Get bIsWaterJump() -> (bIsWaterJump:bool)))
   > Set Component Linear Acc Scale(Component Linear Acc Scale:Vector=@n6) -> (Output_Get:Vector)
   > Set Component Linear Vel Scale(Component Linear Vel Scale:Vector=@n6) -> (Output_Get:Vector)
   n7 = GetOwningActor()
   n8 = GetVelocity!(self:Actor Object Reference=@n7) -> (ReturnValue:Vector)
   n9 = GetActorForwardVector!(self:Actor Object Reference=@n7) -> (ReturnValue:Vector)
   Set ForwardMovementAlpha(
    execute=@n4.then_1
    ForwardMovementAlpha:double=MapRangeClamped(Value=Dot Product!(A:Vector=@n8, B:Vector=@n9) -> (ReturnValue:double), InRangeA=0, InRangeB=400)
   ) -> (Output_Get:double)
   > Set FloatConstraint_alpha(FloatConstraint_alpha:double=SelectFloat(A=0.4, B=0, bPickA=Get bIsSkydiveFloating())) -> (Output_Get:double)
   n10 = Get SkelMesh()
   Set LeftLegForwardBendCompensationAlpha(
    execute=@n4.then_0
    LeftLegForwardBendCompensationAlpha:double=GetAlphaFromReferenceTransformDelta(SkeletalMeshComponent=@n10, BoneName=thigh_l, BaseName=pelvis, InputRange=(X=-17,Y=-103), TwistAxis="1,0,0")
   ) -> (Output_Get:double)
   > Set RightLegForwardBendCompensationAlpha(RightLegForwardBendCompensationAlpha:double=GetAlphaFromReferenceTransformDelta(SkeletalMeshComponent=@n10, BoneName=thigh_r, BaseName=pelvis, InputRange=(X=-17,Y=-103), TwistAxis="1,0,0")) -> (Output_Get:double)
   > n11 = Set RThighFrontUpAlpha(RThighFrontUpAlpha:double=GetAlphaFromReferenceTransformDelta(SkeletalMeshComponent=@n10, BoneName=thigh_r, BaseName=pelvis, InputRange=(X=-17,Y=-60), TwistAxis="1,0,0")) -> (Output_Get:double)
   > Set EMOTE_RThighFrontUp_FPS_Correction_Alpha(EMOTE_RThighFrontUp_FPS_Correction_Alpha:double=float * float!(A=@n11, B=Get FPS Compensation Alpha() -> (FPS Compensation Alpha:double), C:bool=Get bIsPlayingEmote()) -> (ReturnValue) {is=title K2Node_PromotableOperator}) -> (Output_Get:double)
   > n12 = Set LThighFrontUpAlpha(LThighFrontUpAlpha:double=GetAlphaFromReferenceTransformDelta(SkeletalMeshComponent=@n10, BoneName=thigh_l, BaseName=pelvis, InputRange=(X=0,Y=-60), TwistAxis="1,0,0")) -> (Output_Get:double)
   > Set EMOTE_LThighFrontUp_FPS_Correction_Alpha(EMOTE_LThighFrontUp_FPS_Correction_Alpha:double=float * float(A=@n12, B=Get FPS Compensation Alpha(), C=Get bIsPlayingEmote())) -> (Output_Get:double)
   > Set RightLegSideBendAlpha(RightLegSideBendAlpha:float=GetAlphaFromReferenceTransformDelta(
     SkeletalMeshComponent=Get SkelMesh()
     BoneName=thigh_r
     BaseName=pelvis
     DeltaComponentType=EulerY
     InputRange=(X=20,Y=45)
     TwistAxis="1,0,0"
    )) -> (Output_Get:float)
   > Set IsJumpingAlpha(IsJumpingAlpha:float="To Float (Boolean)"(InBool=Get bIsJumping())) -> (Output_Get:float)
   > Set IsMeleeAlpha(IsMeleeAlpha:float="To Float (Boolean)"(InBool=Get bIsPlayingMeleeAnim())) -> (Output_Get:float)
   n13 = Get bIsMoving2D()
   n14 = Get bIsCrouching()
   > Set CrouchingMovingAlpha(CrouchingMovingAlpha:float="To Float (Boolean)"(InBool=AND Boolean(A=@n13, B=@n14))) -> (Output_Get:float)
   n15 = Get bIsInFrontEnd()
   > Set IsNotInFrontEnd(IsNotInFrontEnd:float="To Float (Boolean)"(InBool=NOT Boolean(A=@n15))) -> (Output_Get:float)
   n16 = "To Float (Boolean)"(InBool=Get bIsInFrontEnd())
   Set FE L LegFwdEmote(
    execute=@n4.then_3
    FE L LegFwdEmote:double=float * float!(A=@n16, B=Get EMOTE_LThighFrontUp_FPS_Correction_Alpha() -> (EMOTE_LThighFrontUp_FPS_Correction_Alpha:double)) -> (ReturnValue)
   ) -> (Output_Get:double)
   > Set FE R LegFwdEmote(FE R LegFwdEmote:double=float * float(A=@n16, B=Get EMOTE_RThighFrontUp_FPS_Correction_Alpha() -> (EMOTE_RThighFrontUp_FPS_Correction_Alpha:double))) -> (Output_Get:double)
   n17 = Set FPS Compensation Alpha(execute=@n2.else, FPS Compensation Alpha=MapRangeClamped(Value=@n1.DeltaTimeX, InRangeA=0.0166, InRangeB=0.0333))
   Event BlueprintInitializeAnimation()
   > Set SkelMesh(SkelMesh=GetOwnerSkeletalMeshComponent())
   Event OnCharacterCustomizationLoaded()
   > Set SkelMesh(SkelMesh=GetPartSkeletalMeshComponent())
  ]}
]
variables [
 Component Linear Acc Scale {type=Vector}
 Component Linear Vel Scale {type=Vector}
 CrouchingMovingAlpha {type=number}
 EMOTE_LThighFrontUp_FPS_Correction_Alpha {type=number}
 EMOTE_RThighFrontUp_FPS_Correction_Alpha {type=number}
 FE L LegFwdEmote {type=number}
 FE R LegFwdEmote {type=number}
 FloatConstraint_alpha {type=number}
 ForwardMovementAlpha {type=number}
 FPS Compensation Alpha {type=number}
 Gravity Override {type=Vector}
 Gravity Override Params {
  type=GravityOverrideParamsStruct
  default {
   skydive_FloatingDown {gravityOverride=<0 0 -900>, windAmplitude=<20 -50 -50>, windFrequency=<1.5 3 3>, windOffset=<0 -100 -100>}
   skydive_GoingUp {gravityOverride=<0 0 -800>, windAmplitude=<20 -50 -49>, windOffset=<0 -100 -100>}
  }
 }
 IsJumpingAlpha {type=number}
 IsMeleeAlpha {type=number}
 IsNotInFrontEnd {type=number}
 LeftLegForwardBendCompensationAlpha {type=number}
 LThighFrontUpAlpha {type=number}
 RBAN Params {
  type=RBANParamsStruct
  default {
   dBNO {simSpaceSettings {
     maxAngularAcceleration=1000
     maxAngularVelocity=3
     maxLinearAcceleration=300
     maxLinearVelocity=700
     velocityScaleZ=0
     externalLinearDragV=<0 0.5 0>
    }}
   dBNO_Carried {simSpaceSettings {
     maxAngularAcceleration=1000
     maxAngularVelocity=3
     maxLinearAcceleration=300
     maxLinearVelocity=700
     velocityScaleZ=0
     externalLinearDragV=<0 0.5 0>
    }}
   emoteOrMelee {simSpaceSettings {
     maxAngularAcceleration=1000
     maxAngularVelocity=3
     maxLinearAcceleration=300
     maxLinearVelocity=700
     velocityScaleZ=0
     externalLinearDragV=<0 0.5 0>
    }}
   front_End {simSpaceSettings {
     maxAngularAcceleration=1000
     maxAngularVelocity=3
     maxLinearAcceleration=300
     maxLinearVelocity=700
     velocityScaleZ=0
     externalLinearDragV=<0 0.5 0>
    }}
   onGround_Crouch_Moving {simSpaceSettings {
     maxAngularAcceleration=1000
     maxAngularVelocity=3
     maxLinearAcceleration=300
     maxLinearVelocity=700
     velocityScaleZ=0
     externalLinearDragV=<0 0.5 0>
    }}
   onGround_Falling {simSpaceSettings {
     maxAngularAcceleration=1000
     maxAngularVelocity=3
     maxLinearAcceleration=300
     maxLinearVelocity=700
     velocityScaleZ=0
     externalLinearDragV=<0 0.5 0>
    }}
   onGround_Moving {simSpaceSettings {
     maxAngularAcceleration=1000
     maxAngularVelocity=3
     maxLinearAcceleration=300
     maxLinearVelocity=700
     velocityScaleZ=0
     externalLinearDragV=<0 0.5 0>
    }}
   onGround_Standing {simSpaceSettings {
     maxAngularAcceleration=1000
     maxAngularVelocity=3
     maxLinearAcceleration=300
     maxLinearVelocity=700
     velocityScaleZ=0
     externalLinearDragV=<0 0.5 0>
    }}
   rail_Grinding {simSpaceSettings {
     maxAngularAcceleration=1000
     maxAngularVelocity=3
     maxLinearAcceleration=300
     maxLinearVelocity=700
     velocityScaleZ=0
     worldAlpha=0.1
     externalLinearDragV=<0 0.5 0>
    }}
   riding_Moving {simSpaceSettings {
     maxAngularAcceleration=1000
     maxAngularVelocity=3
     maxLinearAcceleration=300
     maxLinearVelocity=700
     velocityScaleZ=0
     externalLinearDragV=<0 0.5 0>
    }}
   riding_Moving_Targeting {simSpaceSettings {
     maxAngularAcceleration=1000
     maxAngularVelocity=3
     maxLinearAcceleration=300
     maxLinearVelocity=700
     velocityScaleZ=0
     externalLinearDragV=<0 0.5 0>
    }}
   riding_Standing {simSpaceSettings {
     maxAngularAcceleration=1000
     maxAngularVelocity=3
     maxLinearAcceleration=300
     maxLinearVelocity=700
     velocityScaleZ=0
     externalLinearDragV=<0 0.5 0>
    }}
   skydive_DivingDown {simSpaceSettings {
     maxAngularAcceleration=1000
     maxAngularVelocity=3
     maxLinearAcceleration=300
     maxLinearVelocity=700
     velocityScaleZ=0
     externalLinearDragV=<0 0.5 0>
    }}
   skydive_FloatingDown {
    componentLinearAccScale=<0 0.5 0>
    componentLinearVelScale=<0 0.5 0>
    simSpaceSettings {
     maxAngularAcceleration=1000
     maxAngularVelocity=3
     maxLinearAcceleration=300
     maxLinearVelocity=700
     velocityScaleZ=0
     externalLinearDragV=<0 0.5 0>
    }
   }
   skydive_GoingUp {simSpaceSettings {
     maxAngularAcceleration=1000
     maxAngularVelocity=3
     maxLinearAcceleration=300
     maxLinearVelocity=700
     velocityScaleZ=0
     externalLinearDragV=<0 0.5 0>
    }}
   skydive_Parachuting {
    componentLinearAccScale=<0 0.1 0.3>
    componentLinearVelScale=<0 0.1 0.3>
    simSpaceSettings {
     maxAngularAcceleration=1000
     maxAngularVelocity=3
     maxLinearAcceleration=300
     maxLinearVelocity=700
     velocityScaleZ=0
     externalLinearDragV=<0 0.5 0>
    }
   }
   skydive_Surfing {simSpaceSettings {
     maxAngularAcceleration=1000
     maxAngularVelocity=3
     maxLinearAcceleration=300
     maxLinearVelocity=700
     velocityScaleZ=0
     externalLinearDragV=<0 0.5 0>
    }}
   swimming {simSpaceSettings {
     maxAngularAcceleration=1000
     maxAngularVelocity=3
     maxLinearAcceleration=300
     maxLinearVelocity=700
     velocityScaleZ=0
     externalLinearDragV=<0 0.5 0>
    }}
   tactical_Sprint {simSpaceSettings {
     maxAngularAcceleration=1000
     maxAngularVelocity=3
     maxLinearAcceleration=300
     maxLinearVelocity=700
     velocityScaleZ=0
     externalLinearDragV=<0 0.5 0>
    }}
   ziplining {simSpaceSettings {
     maxAngularAcceleration=1000
     maxAngularVelocity=3
     maxLinearAcceleration=300
     maxLinearVelocity=700
     velocityScaleZ=0
     externalLinearDragV=<0 0.5 0>
    }}
  }
 }
 RightLegForwardBendCompensationAlpha {type=number}
 RightLegSideBendAlpha {type=number}
 RThighFrontUpAlpha {type=number}
 SkelMesh {type=SkeletalMeshComponent}
]
