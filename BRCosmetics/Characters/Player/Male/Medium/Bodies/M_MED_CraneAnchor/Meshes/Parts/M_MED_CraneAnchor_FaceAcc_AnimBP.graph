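# Post-processes the CraneAnchor male face-accessory pose, adding secondary hair motion to the pose copied from the attached parent mesh.
# Chained AnimDynamics simulate the master hair bone and right/left three-bone strands: angular limits on A/B bones, half-alpha linear springs on C tips.
# Two additive bone-space yaws on mstr_hair_b follow, largely cancelling each other.
# The event graph computes HeadLookUpAlpha from head-versus-spine_05 rotation, but no shown node reads it.
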AnimBlueprint /BRCosmetics/Characters/Player/Male/Medium/Bodies/M_MED_CraneAnchor/Meshes/Parts/M_MED_CraneAnchor_FaceAcc_AnimBP
target_skeleton=&/M_MED_CraneAnchor_FaceAcc_Skeleton
graphs [
 AnimGraph {nodes [
   n1 = Anim Dynamics - Bone: mstr_hair_b(ComponentPose=Local To Component(LocalPose=Copy Pose From Mesh() {bUseAttachedParent=true}), Alpha=1, ExternalForce="0,0,0") {
    angularBiasOverride=1
    angularDampingOverride=0.999
    angularSpringConstant=1500
    bAngularSpring=true
    bGravityOverrideInSimSpace=true
    bOverrideAngularDamping=true
    gravityScale=0.35
    simulationSpace=RootRelative
    alphaBoolBlend {blendOption=Linear}
    boundBone {boneName=mstr_hair_b}
    componentAppliedLinearAccClamp=<3000 3000 3000>
    componentLinearAccScale=<0.3 0.3 0.3>
    componentLinearVelScale=<0.3 0.3 0.3>
    physicsBodyDefinitions [{
      sphereCollisionRadius=0
      boundBone {boneName=mstr_hair_b}
      boxExtents=<15 20 20>
      constraintSetup {angularLimitsMax=<0 10 20>, angularLimitsMin=<0 -10 0>, angularTarget=<0 0 0>}
      localJointOffset=<-5 0 0>
     }]
   }
   n2 = Anim Dynamics - Bone: dyn_hair_r_A(ComponentPose=@n1, Alpha=1, ExternalForce="0,0,0") {
    angularBiasOverride=1
    angularDampingOverride=0.999
    angularSpringConstant=1500
    bAngularSpring=true
    bGravityOverrideInSimSpace=true
    bOverrideAngularDamping=true
    gravityScale=0.35
    simulationSpace=RootRelative
    alphaBoolBlend {blendOption=Linear}
    boundBone {boneName=dyn_hair_r_A}
    componentAppliedLinearAccClamp=<3000 3000 3000>
    componentLinearAccScale=<0.3 0.3 0.3>
    componentLinearVelScale=<0.3 0.3 0.3>
    physicsBodyDefinitions [{
      sphereCollisionRadius=0
      boundBone {boneName=dyn_hair_r_A}
      boxExtents=<15 20 20>
      constraintSetup {angularLimitsMax=<0 0 20>, angularLimitsMin=<0 -10 -20>, angularTarget=<0 0 0>}
      localJointOffset=<-5 0 0>
     }]
   }
   n3 = Anim Dynamics - Bone: dyn_hair_r_B(ComponentPose=@n2, Alpha=1, ExternalForce="0,0,0") {
    angularBiasOverride=1
    angularDampingOverride=1
    angularSpringConstant=1500
    bAngularSpring=true
    bGravityOverrideInSimSpace=true
    bOverrideAngularBias=true
    bOverrideAngularDamping=true
    gravityScale=0.35
    simulationSpace=RootRelative
    alphaBoolBlend {blendOption=Linear}
    boundBone {boneName=dyn_hair_r_B}
    componentAppliedLinearAccClamp=<3000 3000 3000>
    componentLinearAccScale=<0.3 0.3 0.3>
    componentLinearVelScale=<0.3 0.3 0.3>
    physicsBodyDefinitions [{
      sphereCollisionRadius=0
      boundBone {boneName=dyn_hair_r_B}
      boxExtents=<15 20 20>
      constraintSetup {angularLimitsMax=<0 0 10>, angularLimitsMin=<0 -5 -10>, angularTarget=<0 0 0>}
      localJointOffset=<-1 0 0>
     }]
   }
   n4 = Anim Dynamics - Bone: dyn_hair_r_C(ComponentPose=@n3, Alpha=0.5, ExternalForce="0,0,0") {
    bLinearSpring=true
    bOverrideLinearDamping=true
    bUseGravityOverride=true
    linearDampingOverride=1
    linearSpringConstant=175
    simulationSpace=RootRelative
    alphaBoolBlend {blendOption=Linear}
    boundBone {boneName=dyn_hair_r_C}
    componentAppliedLinearAccClamp=<2500 2500 2500>
    componentLinearAccScale=<0.2 0.2 0.2>
    componentLinearVelScale=<0.9 0.9 0.9>
    physicsBodyDefinitions [{
      boundBone {boneName=dyn_hair_r_C}
      boxExtents=<60 60 60>
      constraintSetup {angularLimitsMax=<0 0 0>, angularLimitsMin=<0 0 0>, angularTarget=<0 0 0>, linearAxesMax=<1 0.5 0.5>, linearAxesMin=<-0.5 -0.5 -0.5>}
      localJointOffset=<0 0 0>
     }]
   }
   n5 = Anim Dynamics - Bone: dyn_hair_l_A(ComponentPose=@n4, Alpha=1, ExternalForce="0,0,0") {
    angularBiasOverride=1
    angularDampingOverride=0.999
    angularSpringConstant=1500
    bOverrideAngularDamping=true
    bUseGravityOverride=true
    gravityScale=0.35
    simulationSpace=RootRelative
    alphaBoolBlend {blendOption=Linear}
    boundBone {boneName=dyn_hair_l_A}
    componentAppliedLinearAccClamp=<30000 30000 30000>
    componentLinearAccScale=<0.2 0.2 0.2>
    componentLinearVelScale=<0.2 0.2 0.2>
    gravityOverride {z=-980}
    physicsBodyDefinitions [{
      sphereCollisionRadius=0
      boundBone {boneName=dyn_hair_l_A}
      boxExtents=<15 20 20>
      constraintSetup {angularTargetAxis=AxisY, angularLimitsMax=<0 10 20>, angularLimitsMin=<0 0 -20>, angularTarget=<0 -10 0>}
      localJointOffset=<-5 0 0>
     }]
   }
   n6 = Anim Dynamics - Bone: dyn_hair_l_B(ComponentPose=@n5, Alpha=1, ExternalForce="0,0,0") {
    angularBiasOverride=1
    angularDampingOverride=1
    angularSpringConstant=1500
    bOverrideAngularBias=true
    bOverrideAngularDamping=true
    bUseGravityOverride=true
    gravityScale=0.35
    simulationSpace=RootRelative
    alphaBoolBlend {blendOption=Linear}
    boundBone {boneName=dyn_hair_l_B}
    componentAppliedLinearAccClamp=<30000 30000 30000>
    componentLinearAccScale=<0.2 0.2 0.2>
    componentLinearVelScale=<0.2 0.2 0.2>
    gravityOverride {z=-980}
    physicsBodyDefinitions [{
      sphereCollisionRadius=0
      boundBone {boneName=dyn_hair_l_B}
      boxExtents=<15 20 20>
      constraintSetup {angularTargetAxis=AxisY, angularLimitsMax=<0 10 10>, angularLimitsMin=<0 0 -10>, angularTarget=<0 -10 0>}
      localJointOffset=<-1 0 0>
     }]
   }
   n7 = Anim Dynamics - Bone: dyn_hair_l_C(ComponentPose=@n6, Alpha=0.5, ExternalForce="0,0,0") {
    bLinearSpring=true
    bOverrideLinearDamping=true
    bUseGravityOverride=true
    linearDampingOverride=1
    linearSpringConstant=175
    simulationSpace=RootRelative
    alphaBoolBlend {blendOption=Linear}
    boundBone {boneName=dyn_hair_l_C}
    componentAppliedLinearAccClamp=<2500 2500 2500>
    componentLinearAccScale=<0.2 0.2 0.2>
    componentLinearVelScale=<0.9 0.9 0.9>
    physicsBodyDefinitions [{
      boundBone {boneName=dyn_hair_l_C}
      boxExtents=<60 60 60>
      constraintSetup {angularLimitsMax=<0 0 0>, angularLimitsMin=<0 0 0>, angularTarget=<0 0 0>, linearAxesMax=<1 0.5 0.5>, linearAxesMin=<-0.5 -0.5 -0.5>}
      localJointOffset=<0 0 0>
     }]
   }
   n8 = "Transform (Modify) Bone - Bone: mstr_hair_b"(ComponentPose=@n7, Alpha=1, Translation="0,0,0", Rotation="0,45,0", Scale="1,1,1") {
    rotationMode=BMM_Additive
    rotationSpace=BCS_BoneSpace
    translationMode=BMM_Additive
    alphaBoolBlend {blendOption=Linear}
    boneToModify {boneName=mstr_hair_b}
    rotation {yaw=45}
   }
   n9 = "Transform (Modify) Bone - Bone: mstr_hair_b"(ComponentPose=@n8, Alpha=1, Translation="0,0,0", Rotation="0,-35,0", Scale="1,1,1") {
    rotationMode=BMM_Additive
    rotationSpace=BCS_BoneSpace
    translationMode=BMM_Additive
    alphaBoolBlend {blendOption=Linear}
    boneToModify {boneName=mstr_hair_b}
    rotation {yaw=-35}
   }
   Output Pose(Result=Component To Local(ComponentPose=@n9)) {layerGroup=None}
  ]}
 EventGraph {nodes [
   Event BlueprintUpdateAnimation()
   > Set HeadLookUpAlpha(HeadLookUpAlpha:float=GetAlphaFromReferenceTransformDelta(SkeletalMeshComponent=Get SkelMesh(), BoneName=head, BaseName=spine_05, InputRange=(X=-5,Y=-40), TwistAxis="1,0,0")) -> (Output_Get:float)
   Event BlueprintInitializeAnimation()
   > Set SkelMesh(SkelMesh=GetOwnerSkeletalMeshComponent())
  ]}
]
variables [
 HeadLookUpAlpha {type=number}
 SkelMesh {type=SkeletalMeshComponent}
]
